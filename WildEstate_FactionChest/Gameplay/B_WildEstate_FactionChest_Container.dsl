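; Faction chest container Blueprint for Wild Estate that drives the chest's idle and opening visual effects.
; BeginPlay shows the Fading FX component with its custom primitive data reset to zero.
; On open, it spawns a Niagara burst tinted by the container's highest-rarity color.
; A sped-up Idle FX timeline simultaneously animates that primitive data, setting it to one and hiding the component when finished.
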(event EventBeginPlay
  (bind _fading_fx (Variables|Default|GetFading%20FX))
  (|CallParentFunction)
  (Rendering|Material|SetCustomPrimitiveDataFloat _fading_fx 2 0.0)
  (Rendering|SetVisibility _fading_fx true false))

(event EventSpawnChestOpenVFX
  (bind _fading_fx (Variables|Default|GetFading%20FX))
  (bind _returnvalue (Niagara|SpawnSystematLocation "/WildEstate_FactionChest/Assets/FX/NS_ChestBurst_Normal_toon.NS_ChestBurst_Normal_toon" (Transformation|GetActorLocation) (Transformation|GetActorRotation)))
  (bind _returnvalue_2 (Fortnite|GetRarityData (Variables|Container|GetHighestRarity)))
  (Effects|Components|ParticleSystem|SetVectorParameter _returnvalue "Color" (Math|Conversions|ToVector(LinearColor) (.Color2 (Utilities|Struct|BreakFortRarityItemData _returnvalue_2))))
  (Components|Timeline|SetPlayRate (Variables|B_WildEstate_FactionChest_Container|GetIdle%20FX) (/ 1.0 0.35))
  (bind (_direction _newtrack_0) (Timeline|Idle%20FX PlayFromStart)
    (:Update
      (Rendering|Material|SetCustomPrimitiveDataFloat _fading_fx 2 _newtrack_0))
    (:Finished
      (Rendering|Material|SetCustomPrimitiveDataFloat _fading_fx 2 1.0)
      (Rendering|SetVisibility _fading_fx false false))))

(event Collision|EventActorBeginOverlap
  (params (OtherActor :type Actor%20Object%20Reference)))

(event EventTick
  (params (DeltaSeconds :type Float%20%28single-precision%29)))

(fn ConstructionScript
  (|CallParentFunction))
